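# Sound cue for the LimaBean back bling, played on DBNO (down but not out), layering an impostor kill music sting with a fall foley sound.
# After a short fixed delay, a mixer combines the music with the pitched-up, boosted foley.
# An IsLocal branch gives the music stereo attenuation for the local player and ambio attenuation otherwise or when unset.
# Concurrency allows one instance per owner, blocking new plays.
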SoundCue /BRCosmetics/Sounds/Backblings/LimaBean/BackBling_LimaBean_DBNO_Cue
root=Delay_2
nodes [
 {
  id=Attenuation_0
  children [Mixer_0]
  properties {
   attenuationSettings=/BRCosmetics/Athena/Sounds/BackBlings/LimaBean/ATT_Fort_LimaBean_Ambio
   attenuationOverrides {bEnableReverbSend=false, bSpatialize=false, falloffDistance=300, attenuationShapeExtents {x=250}}
  }
 }
 {
  id=Attenuation_1
  children [WavePlayer_1]
  properties {
   attenuationSettings=&/Fort_LimaBean_Stereo_Att
   attenuationOverrides {bEnableReverbSend=false, bSpatialize=false, falloffDistance=300, attenuationShapeExtents {x=250}}
  }
 }
 {
  id=Attenuation_2
  children [Modulator_0]
  properties {
   attenuationSettings=&ATT_Fort_LimaBean_Ambio
   attenuationOverrides {bEnableReverbSend=false, bSpatialize=false, falloffDistance=300, attenuationShapeExtents {x=250}}
  }
 }
 {id=Branch_0, children [Attenuation_1, Attenuation_0, Attenuation_0], properties {boolParameterName=IsLocal}}
 {id=Delay_2, children [Mixer_1], properties {delayMax=0.25, delayMin=0.25}}
 {id=Mixer_0, children [WavePlayer_0], properties {inputVolume [1]}}
 {id=Mixer_1, children [Branch_0, Attenuation_2], properties {inputVolume [1,3]}}
 {id=Modulator_0, children [WavePlayer_2], properties {pitchMax=1.2, pitchMin=1.2, volumeMax=1, volumeMin=1}}
 {id=WavePlayer_0, properties {soundWaveAssetPtr=&/impostor_killMusic}}
 {id=WavePlayer_1, properties {soundWaveAssetPtr=&impostor_killMusic}}
 {id=WavePlayer_2, properties {soundWaveAssetPtr=/BRCosmetics/Athena/Sounds/BackBlings/LimaBean/BackBling_LimaBean_Fall_Foley_01}}
]
properties {
 audioPropertiesSheet=/BRCosmetics/Sounds/Settings/PropertySheets/Reactive/APS_Cosmetic_Reactive_Backbling
 maxDistance=1800.0001
 soundClassObject=/Game/Sounds/FortSoundClassesAndModes/Fort_Backblings
 volumeMultiplier=0.65
 concurrencyOverrides {bLimitToOwner=true, maxCount=1, resolutionRule=PreventNew}
}
